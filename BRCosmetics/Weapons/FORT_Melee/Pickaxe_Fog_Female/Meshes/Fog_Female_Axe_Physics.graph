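# Physics asset for the Fog Female pickaxe mesh, a two-handed melee cosmetic with a dynamic ring bone.
# A capsule covers the handle bone, and a zero-length capsule gives the dyn_ring bone an effectively spherical collision shape.
# One constraint joins the ring to the handle with zero swing1 and twist limits and a wide swing2 limit, so the ring swings like a hinge.
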PhysicsAsset /BRCosmetics/Weapons/FORT_Melee/Pickaxe_Fog_Female/Meshes/Fog_Female_Axe_Physics
bodies {
 Melee_TwoHanded_Handle [{
   shapeType=Capsule
   length=23.3731
   radius=3.0352
   shapeName=Melee_TwoHanded_Handle_capsule
   center=<-1.0989 1.3601 1.183>
   rotation=<rot -0.1855 -90 0.0562>
  }]
 dyn_ring [{shapeType=Capsule, length=0, radius=4.7368, shapeName=dyn_ring_capsule, center=<6.0144 0 0>, rotation=<rot 90 0 0>}]
}
constraints [{bone1Name=dyn_ring, bone2Name=Melee_TwoHanded_Handle, swing1LimitDegrees=0, swing2LimitDegrees=75, twistLimitDegrees=0}]
